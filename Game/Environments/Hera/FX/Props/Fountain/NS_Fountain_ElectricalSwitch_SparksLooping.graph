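# Looping Niagara effect for electrical sparks bursting from a fountain's electrical switch prop.
# One emitter loops with randomized duration; a loop-count seed picks one of twelve position and direction pairs per burst.
# Each burst spawns three spawn groups: falling sparks, a stationary flash and slowly rising, smoke-like puff, with visibility tags matching groups to renderers.
# Three sprite renderers and a component renderer draw them, with distance-based size scaling.
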NiagaraSystem /Game/Environments/Hera/FX/Props/Fountain/NS_Fountain_ElectricalSwitch_SparksLooping
dependencies {
 usedDataInterfaces [
  &NiagaraDataInterfaceAudioPlayer
  &NiagaraDataInterfaceArrayFloat3
  &NiagaraDataInterfaceCurve
  &NiagaraDataInterfaceColorCurve
  &NiagaraDataInterfaceSpriteRendererInfo
 ]
 usedDynamicInputs [&RandomRangeVector, &MakeQuaternion]
 usedModules [&SystemState]
}
emitters {Sparks {
  emitterSpawnScript [&InitializeAudioPlayer()]
  emitterUpdateScript [
   &EmitterState(
    Delay First Loop Only=true
    Life Cycle Mode=::NewEnumerator1
    Loop Delay=0.5
    Loop Duration=&RandomRangeFloat(Minimum=0.1, Maximum=1, Evaluation Type=::NewEnumerator1)
    Recalculate Duration Each Loop=true
    UseLoopDelay=true
   )
   &SpawnBurst_Instantaneous(Spawn Group=1, Use Spawn Probability=false, UseLoopCountLimit=false) {module=SpawnBurst_Instantaneous001}
   &SpawnBurst_Instantaneous(
    Spawn Count=&RandomRangeInt(Minimum=5, Maximum=15)
    Use Spawn Probability=false
    UseLoopCountLimit=false
   )
   &SpawnBurst_Instantaneous(Spawn Group=2, Use Spawn Probability=false, UseLoopCountLimit=false) {module=SpawnBurst_Instantaneous002}
  ]
  particleSpawnScript [
   /SetVariables(Particles.Randomness:NiagaraInt32=&RandomRangeInt(Minimum=0, Maximum=11, Randomness Mode=::NewEnumerator1, Random Seed=$Emitter.LoopCount, Fixed Random Seed=true)) {isSetParameters=true}
   /SetVariables(Particles.VisibilityTag=&SpawnGroup()) {isSetParameters=true}
   &InitializeParticle(
    Color=<rgba 50 50 50 1>
    Lifetime=&MaskFloat(
     Mask Source=0
     Mask Parameter=?
     Mask=2
     Masked Float=&RandomRangeFloat(Minimum=1, Maximum=1.5)
     Passthrough Float=&RandomRangeFloat(Minimum=0.25, Maximum=0.5)
    )
    Sprite Rotation Mode=::NewEnumerator2
    Sprite Size Mode=::NewEnumerator1
    Sprite UV Mode=::NewEnumerator3
    Uniform Sprite Size Max=4.5
    Uniform Sprite Size Min=2.25
    UsePositionOffset=false
    Write Material Random=true
   )
   /SetVariables(Particles.Position=&SelectVectorFromArray(
     Direct Array Index=$Particles.Randomness
     Vector Selection Array=DataInterface() {propertyValues {FloatData [
        <-150 69 105>
        <-77 69 86>
        <-90 69 -64>
        <-210 -92 40>
        <-210 48 97>
        <-210 48 -46>
        <-210 -47 108>
        <-16 -82 108>
        <-16 48 48>
        <3 15 23>
        <3 3 -86>
        <3 20 -104>
       ]}}
    )) {isSetParameters=true}
   &AddVelocity(
    Cone Angle=&RandomRangeFloat(Minimum=15, Maximum=25)
    Cone Axis=&SelectVectorFromArray(
     Direct Array Index=$Particles.Randomness
     Vector Selection Array=DataInterface() {propertyValues {FloatData [
        <0 1 0.25>
        <0.1 1 0.25>
        <-0.2 0.5 0.7>
        <-1 -1 1>
        <-1 0.5 0.5>
        <-1 1 0.2>
        <-1 -0.5 1>
        <1 -0.5 0.7>
        <1 0.8 0.3>
        <1 0 1>
        <1 -0.5 1>
        <0.5 1 1>
       ]}}
    )
    Module.UseConeFalloff=false
    Module.UseVelDistribution=false
    Override Seed=false
    Velocity Mode=::NewEnumerator2
    Velocity Speed=&MaskFloat(
     Mask Source=0
     Mask Parameter=?
     Mask=0
     Masked Float=&RandomRangeFloat(Minimum=75, Maximum=350)
     Passthrough Float=0
    )
   )
  ]
  particleUpdateScript [
   &ParticleState()
   &GravityForce(Gravity=&Multiply_VectorByFloat(
     Vector=&SelectVectorFromArray(
      Direct Array Index=&SpawnGroup()
      Vector Selection Array=DataInterface() {propertyValues {FloatData [<0 0 -490>, <0 0 0>, <0 0 50>]}}
     )
     Float=&Multiply_Float(A=1, B=1.2)
    ))
   &Drag(Use Linear Drag=true, Use Rotational Drag=true)
   &SolveForcesAndVelocity(Clamp Velocity=false, Limit Acceleration=false)
   &ScaleSpriteSizeBySpeed(Max Scale Factor=<0.25 1>, Min Scale Factor=<0.05 0.05>, Velocity Threshold=100)
   &ScaleSpriteSize(
    Scale Sprite Size Mode=::NewEnumerator0
    Uniform Scale Factor=&FN_LOD_Distance_Remap(Start Distance=0, End Distance=3500, Minimum Output=1, Maximum Output=2, Clamp Output=true)
   )
   &ScaleColor(
    Scale RGB=&MakeCustomVectorFromBool(
     Bool=&RandomBool()
     True Vector=&MakeVectorFromLinearColorRGB(Color=<rgba 5 0.8333 0 1>)
     False Vector=&MakeVectorFromLinearColorRGB(Color=<rgba 0 0.9566 5 1>)
    )
    ScaleA=true
    ScaleRGB=true
   )
   &FN_Sprite_Attributes(
    Camera Offset=50
    Color=&Multiply_LinearColorByFloat(
     Linear Color=&ColorFromCurve(ColorCurve=DataInterface() {propertyValues {
        AlphaCurve {keys [<at 0 0.1 RCIM_Cubic>,<at 0.5886 0 RCIM_Cubic>]}
        BlueCurve {keys [<at 0 3 RCIM_Cubic>,<at 0.2098 1 RCIM_Cubic>]}
        GreenCurve {keys [<at 0 0.3407 RCIM_Cubic>,<at 0.2098 0.2468 RCIM_Cubic>]}
        RedCurve {keys [<at 0 0 RCIM_Cubic>,<at 0.2098 0 RCIM_Cubic>]}
       }})
     Float=5
    )
    Create Camera Offset Attribute=true
    Create Color Attribute=true
    "Create Normalized Age "=true
    Create SpriteSize Attribute=true
    Max Age=1
    Module Output:ENiagara_AttributeSamplingApplyOutput=::NewEnumerator1
    SpriteSize=&Multiply_Vector2D_ByFloat(
     Float=&FloatFromCurve(FloatCurve=DataInterface() {propertyValues {Curve {keys [
          {interpMode=RCIM_Cubic, leaveTangent=2.6917, tangentMode=RCTM_Break, time=0, value=0}
          {interpMode=RCIM_Cubic, leaveTangent=-0.1828, tangentMode=RCTM_Break, time=0.125, value=1}
          {interpMode=RCIM_Cubic, leaveTangent=1.2118, tangentMode=RCTM_Break, time=0.25, value=0}
         ]}}})
     Vector2D=<75 75>
    )
   )
   &FN_Sprite_Attributes(
    Camera Offset=75
    Color=&MultiplyLinearColor(
     Linear Color=<rgba 0.025 0.025 0.025 0.7>
     Scale Factor=&MakeVector4_FromLinearColor(LinearColor=&ColorFromCurve(ColorCurve=DataInterface() {propertyValues {
         AlphaCurve {keys [<at 0.0015 0>,<at 0.1039 0.3>,<at 0.9634 0>]}
         BlueCurve {keys [<at 0.0407 15>,<at 0.1389 1>]}
         GreenCurve {keys [<at 0.0407 5.6527>,<at 0.1389 1>]}
         RedCurve {keys [<at 0.0407 0>,<at 0.1389 1>]}
        }}))
    )
    Create Camera Offset Attribute=true
    Create Color Attribute=true
    Create Dynamic Material Parameter Attributes=true
    Create Sprite Rotation Attribute=true
    Create SpriteSize Attribute=true
    Index 0 Param 1:NiagaraFloat=&FloatFromCurve(FloatCurve=DataInterface() {propertyValues {Curve {keys [{interpMode=RCIM_Cubic, leaveTangent=2.8262, tangentMode=RCTM_Break, time=0, value=1}, <at 1 2>]}}})
    Module Output:ENiagara_AttributeSamplingApplyOutput=::NewEnumerator1
    Module.Index0Float0Write=true
    Module.Index0Float1Write=true
    Module.Index0Float2Write=true
    Module.Index0Float3Write=true
    SpriteSize=&Multiply_Vector2D_ByFloat(
     Float=&FloatFromCurve(FloatCurve=DataInterface() {propertyValues {Curve {keys [{interpMode=RCIM_Cubic, leaveTangent=2.2404, tangentMode=RCTM_Break, time=0, value=0}, <at 1 1>]}}})
     Vector2D=&Vector2DFromFloat(Value=&RandomRangeFloat(Minimum=120, Maximum=140))
    )
   ) {module=FN_Sprite_Attributes001}
   &SubUVAnimation(
    "Sprite Renderer "=DataInterface() {propertyValues {SpriteRenderer=&self:Sparks_1.Renderer}}
    SubUV Animation Mode:ENiagara_SubUVLookupModeV2=::NewEnumerator1
    SubUV Curve=DataInterface() {propertyValues {Curve {keys [
        {arriveTangent=4.1785, interpMode=RCIM_Cubic, leaveTangent=4.1785, tangentMode=RCTM_User, time=0, value=0}
        {arriveTangent=1.1449, interpMode=RCIM_Cubic, leaveTangent=1.1449, tangentMode=RCTM_User, time=0.2, value=0.6}
        {arriveTangent=0.2086, interpMode=RCIM_Cubic, leaveTangent=0.2086, tangentMode=RCTM_User, time=1, value=1}
       ]}}}
    UseEndFrame=false
    UseStartFrame=false
   )
  ]
  rendererClasses [&NiagaraSpriteRendererProperties, &NiagaraComponentRendererProperties]
  renderers [
   {rendererClass=&NiagaraSpriteRendererProperties}
   {rendererClass=&NiagaraSpriteRendererProperties, rendererIndex=1}
   {rendererClass=&NiagaraSpriteRendererProperties, rendererIndex=2}
   {rendererClass=&NiagaraComponentRendererProperties, rendererIndex=3}
  ]
 }}
summary {
 systemName=NS_Fountain_ElectricalSwitch_SparksLooping
 emitters [{emitterName=Sparks, rendererClasses [&NiagaraSpriteRendererProperties, &NiagaraComponentRendererProperties]}]
 userVariables []
}
