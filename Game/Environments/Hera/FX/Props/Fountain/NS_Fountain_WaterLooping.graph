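# Looping Niagara water effect for a Hera fountain prop: an arcing stream with splashes.
# WaterStream draws a gravity-pulled ribbon with color, alpha and width curves over life.
# Splashes spawns sub-UV sprites whose random value blends location, velocity and drag between a spout spray and a lower upward splash, the latter held back early in the loop by a LoopedAge curve.
# Splash alpha follows time of day; Water_Ripple has no renderer.
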NiagaraSystem /Game/Environments/Hera/FX/Props/Fountain/NS_Fountain_WaterLooping
dependencies {
 usedDataInterfaces [
  &NiagaraDataInterfaceColorCurve
  &NiagaraDataInterfaceCurve
  &NiagaraDataInterfaceSpriteRendererInfo
 ]
 usedDynamicInputs [
  &FNTimeOfDayColor
  &HueShiftLinearColor
  &Multiply_LinearColorByFloat
 ]
 usedModules [
  &SystemState
  &self:SystemSpawnScript.SystemScriptSource.SystemScriptGraph.NiagaraNodeAssignment_2.SetVariables_#1
 ]
}
emitters {
 Splashes {
  emitterUpdateScript [
   &EmitterState()
   &SpawnRate(SpawnRate=32, Use Spawn Probability=false)
  ]
  particleSpawnScript [
   /SetVariables(Particles.Random=&Lerp_Float(
     A=0
     B=&UniformAOrBFloat(A=0, Evaluation Type=::NewEnumerator0, Recalculate Random Each Loop=true, Override Seed=false)
     Alpha=&FloatFromCurve(
      FloatCurve=DataInterface() {propertyValues {Curve {defaultValue=0, keys [<at 0 0 RCIM_Constant>, {interpMode=RCIM_Cubic, tangentMode=RCTM_User, time=0.5, value=1}]}}}
      CurveIndex=$System.LoopedAge
     )
    )) {isSetParameters=true}
   &InitializeParticle(
    Color=<rgba 0.5781 0.8228 1 0.5>
    Lifetime Max=0.9
    Lifetime Min=0.4
    Lifetime Mode=::NewEnumerator1
    Position Offset=<-50 -63 -5>
    Sprite Size Max=<80 70>
    Sprite Size Min=<35 30>
    Sprite Size Mode=::NewEnumerator2
    UsePositionOffset=true
    Write Material Random=true
   )
   &ShapeLocation(
    Hemisphere Distribution=<1 0.5>
    MaskSpawnGroup=false
    Offset=&Lerp_Vector(A=<0 0 0>, B=<-60 0 -115>, Alpha=$Particles.Random)
    Offset Mode=::NewEnumerator0
    Override Seed=false
    Sphere Radius=10
    UseOwnerScale=true
   )
   &AddVelocity(
    Cone Angle=&Lerp_Float(A=40, B=135, Alpha=$Particles.Random)
    Cone Axis=&Lerp_Vector(A=<-1 0 0>, B=<0 0 1>, Alpha=$Particles.Random)
    Module.UseConeFalloff=false
    Module.UseVelDistribution=false
    Override Seed=false
    Velocity Mode=::NewEnumerator2
    Velocity Speed=&Lerp_Float(
     A=&RandomRangeFloat(Minimum=150, Maximum=275)
     B=&RandomRangeFloat(Minimum=85, Maximum=120)
     Alpha=$Particles.Random
    )
   )
   &AddVelocity(Velocity=&Lerp_Vector(A=<0 0 0>, B=&RandomRangeVector(Minimum=<0 0 50>, Maximum=<0 0 250>), Alpha=$Particles.Random)) {module=AddVelocity001}
   &SubUVAnimation(
    End Frame Range Override=3
    Evaluation Type=::NewEnumerator0
    "Sprite Renderer "=DataInterface() {propertyValues {SpriteRenderer=&self:Splashes_0.NiagaraSpriteRendererProperties_0}}
    SubUV Animation Mode:ENiagara_SubUVLookupModeV2=::NewEnumerator2
    UseEndFrame=false
    UseStartFrame=false
   )
   &CameraOffset(Camera Offset Amount=10)
  ]
  particleUpdateScript [
   &ParticleState()
   &GravityForce(Gravity=<0 0 -490>)
   &Drag(Drag=&Lerp_Float(A=5, B=2, Alpha=$Particles.Random), Use Linear Drag=true, Use Rotational Drag=true)
   &SolveForcesAndVelocity(Clamp Velocity=false, Limit Acceleration=false)
   &DynamicMaterialParameters(
    Index 0 Param 1=&FloatFromCurve(
     FloatCurve=DataInterface() {propertyValues {Curve {keys [
         {arriveTangent=-0.5095, interpMode=RCIM_Cubic, leaveTangent=-0.5095, tangentMode=RCTM_User, time=0, value=1}
         {arriveTangent=-1.3585, interpMode=RCIM_Cubic, leaveTangent=-1.3585, tangentMode=RCTM_User, time=1, value=0}
        ]}}}
     Scale Curve=1.5
    )
    Module.Index0Float0Write=true
    Module.Index0Float1Write=true
    Module.Index0Float2Write=true
    Module.Index0Float3Write=true
   )
   &ScaleSpriteSize(Uniform Curve Sprite Scale=DataInterface() {propertyValues {Curve {
       defaultValue=0
       keys [
        {arriveTangent=4.1785, interpMode=RCIM_Cubic, leaveTangent=4.1785, tangentMode=RCTM_User, time=0, value=0}
        {arriveTangent=1.1179, interpMode=RCIM_Cubic, leaveTangent=1.1179, tangentMode=RCTM_User, time=0.2042, value=0.6228}
        {arriveTangent=0.0224, interpMode=RCIM_Cubic, leaveTangent=0.0224, tangentMode=RCTM_User, time=1, value=1}
       ]
      }}})
   &ScaleColor(
    Scale Alpha=&FloatFromCurve(
     FloatCurve=DataInterface() {propertyValues {Curve {keys [
         {arriveTangent=60.0845, interpMode=RCIM_Cubic, leaveTangent=60.0845, tangentMode=RCTM_User, time=0, value=0}
         <at 0.05 1.0113 RCIM_Cubic>
         {arriveTangent=-2.8419, interpMode=RCIM_Cubic, leaveTangent=-2.8419, tangentMode=RCTM_User, time=1, value=0}
        ]}}}
     Scale Curve=&FNTimeOfDayScalar(Night Scale=2, Morning Scale=1, Evening Scale=1)
    )
    ScaleA=true
    ScaleRGB=true
   )
  ]
  rendererClasses [&NiagaraSpriteRendererProperties]
  renderers [{rendererClass=&NiagaraSpriteRendererProperties}]
 }
 WaterStream {
  emitterUpdateScript [
   &EmitterState()
   &SpawnRate(SpawnRate=25, Use Spawn Probability=false)
  ]
  particleSpawnScript [
   &InitializeRibbon(
    Color=<rgba 1 1 1 0.8>
    Lifetime=0.7
    Position Offset:Vector3f=<-40 -63 0>
    Ribbon Width=25
    UsePositionOffset=true
    Write Material Random=true
   )
   &AddVelocity(Velocity=<-100 0 40>, Velocity Speed Scale=2)
  ]
  particleUpdateScript [
   &ParticleState()
   &GravityForce(Gravity=<0 0 -1200>)
   &Drag(Use Linear Drag=true, Use Rotational Drag=true)
   &SolveForcesAndVelocity(Clamp Velocity=false, Limit Acceleration=false)
   &DynamicMaterialParameters(
    Index 0 Param 1=0.5
    Index 0 Param 2=10
    Module.Index0Float0Write=true
    Module.Index0Float1Write=true
    Module.Index0Float2Write=true
    Module.Index0Float3Write=true
   )
   &ScaleColor(
    Scale Alpha=&FloatFromCurve(
     FloatCurve=DataInterface() {propertyValues {Curve {keys [
         {arriveTangent=-1.4341, interpMode=RCIM_Cubic, time=0, value=0}
         {arriveTangent=0.05, interpMode=RCIM_Cubic, leaveTangent=0.05, time=0.1, value=1}
         {interpMode=RCIM_Cubic, tangentMode=RCTM_Break, time=1, value=0.05}
        ]}}}
     Scale Curve=0.7
    )
    Scale RGB=&MakeVectorFromLinearColorRGB(Color=&ColorFromCurve(ColorCurve=DataInterface() {propertyValues {
        AlphaCurve {keys []}
        BlueCurve {keys [{arriveTangent=-1.0847, interpMode=RCIM_Cubic, time=0, value=1}, <at 0.2431 0.3281 RCIM_Cubic>, <at 0.5613 1 RCIM_Cubic>]}
        GreenCurve {keys [
          {arriveTangent=-3.7763, interpMode=RCIM_Cubic, time=0, value=1}
          {arriveTangent=-0.2768, interpMode=RCIM_Cubic, leaveTangent=-0.2768, time=0.2431, value=0.3149}
          <at 0.5613 0.8446 RCIM_Cubic>
         ]}
        RedCurve {keys [
          {arriveTangent=-8.1645, interpMode=RCIM_Cubic, time=0, value=0.9792}
          {arriveTangent=-0.5523, interpMode=RCIM_Cubic, leaveTangent=-0.5523, time=0.2431, value=0.2751}
          <at 0.5613 0.6691 RCIM_Cubic>
         ]}
       }}))
    ScaleA=true
    ScaleRGB=true
   )
   &ScaleRibbonWidth(Ribbon Width Scale=&FloatFromCurve(FloatCurve=DataInterface() {propertyValues {Curve {keys [
         {arriveTangent=-2.1417, interpMode=RCIM_Cubic, leaveTangent=-2.1417, tangentMode=RCTM_User, time=0.1, value=2}
         {arriveTangent=0.1622, interpMode=RCIM_Cubic, leaveTangent=0.1622, tangentMode=RCTM_User, time=1, value=1}
        ]}}}))
  ]
  rendererClasses [&NiagaraRibbonRendererProperties]
  renderers [{rendererClass=&NiagaraRibbonRendererProperties}]
 }
 Water_Ripple {rendererClasses [], renderers []}
}
summary {
 systemName=NS_Fountain_WaterLooping
 emitters [
  {emitterName=WaterStream, rendererClasses [&NiagaraRibbonRendererProperties]}
  {emitterName=Splashes, rendererClasses [&NiagaraSpriteRendererProperties]}
  {emitterName=Water_Ripple, rendererClasses []}
 ]
 userVariables []
}
