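# Niagara system for the water pipe fountain prop in the Hera environment, apparently adding glow and glint effects, judging by its emitter names.
# It holds three emitters, Light, Rays and SparkleStationary, with no renderers listed for any of them.
# System scripts use SystemState and a spawn-time Set Variables assignment, and the system depends on the FNTimeOfDayColor dynamic input.
# It exposes no user variables and lists no data interfaces.
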NiagaraSystem /Game/Environments/Hera/FX/Props/Fountain/NS_Fountain_WaterPipe
dependencies {
 usedDataInterfaces []
 usedDynamicInputs [&FNTimeOfDayColor]
 usedModules [
  &SystemState
  &self:SystemSpawnScript.SystemScriptSource.SystemScriptGraph.NiagaraNodeAssignment_2.SetVariables_#1
 ]
}
emitters {Light {rendererClasses [], renderers []}, Rays {rendererClasses [], renderers []}, SparkleStationary {rendererClasses [], renderers []}}
summary {
 systemName=NS_Fountain_WaterPipe
 emitters [{emitterName=Rays, rendererClasses []}, {emitterName=SparkleStationary, rendererClasses []}, {emitterName=Light, rendererClasses []}]
 userVariables []
}
